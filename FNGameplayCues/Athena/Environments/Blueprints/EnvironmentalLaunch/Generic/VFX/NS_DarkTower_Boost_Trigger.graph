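# Niagara burst effect for triggering a Dark Tower boost in the environmental launch gameplay cues, a short blue, cyan and purple flash directed along +X.
# Two fading mesh cones, a sprite flash and a brief light accompany speed-stretched GPU sparks shot forward from cylinders.
# Short ribbons are laid along Hermite splines by link order and pushed forward, while lens-flare sprites step along X with shrinking size.
# The BIG_RIBBON emitter is disabled.
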NiagaraSystem /FNGameplayCues/Athena/Environments/Blueprints/EnvironmentalLaunch/Generic/VFX/NS_DarkTower_Boost_Trigger
dependencies {
 usedDataInterfaces [&NiagaraDataInterfaceCurve, &NiagaraDataInterfaceColorCurve]
 usedDynamicInputs []
 usedModules [&SystemState]
}
emitters {
 BIG_RIBBON {
  bEnabled=false
  emitterUpdateScript [
   &EmitterState(Life Cycle Mode=::NewEnumerator1, Loop Behavior=::NewEnumerator1, Loop Duration Mode=::NewEnumerator1, UseLoopDelay=false)
   &SpawnRate(SpawnRate=10, Use Spawn Probability=false)
  ]
  particleSpawnScript [&InitializeRibbon(
    Color=<rgba 1 1 1 0.35>
    Lifetime=2
    Module.Write Color=true
    Module.Write Lifetime=true
    Module.Write Mass=true
    Module.Write Material Random=true
    Module.Write Position=false
    Module.Write RibbonTwist=false
    Module.Write RibbonWidth=true
    Ribbon Width=180
   ) {module=InitializeRibbon001}]
  particleUpdateScript [
   &ParticleState()
   /Game/Effects/Niagara/Modules/Curves/FN_HermiteCubicSpline_v1/FN_HermiteCubicSpline(
    Position Blend:NiagaraFloat=$Particles.NormalizedAge
    Position End:Vector3f=&Add_Vector(
     A=&RandomVector(Vector Scale=10, Constrain to Hemisphere=true, Hemisphere Constraint Vector=<0 0 1>)
     B=<600 0 0>
    )
    Position Start:Vector3f=&Add_Vector(
     A=&RandomVector(Vector Scale=0, Constrain to Hemisphere=true, Hemisphere Constraint Vector=<0 0 1>)
     B=<0 0 0>
    )
    Write Particles Position=true
   )
   &ScaleColor(
    Scale Alpha=&FloatFromCurve(
     FloatCurve=DataInterface() {propertyValues {Curve {keys [
         {arriveTangent=-2.2486, interpMode=RCIM_Cubic, time=0, value=0.5}
         {arriveTangent=-0.9875, interpMode=RCIM_Cubic, leaveTangent=-0.9875, time=0.1843, value=0.6348}
         {arriveTangent=-0.1926, interpMode=RCIM_Cubic, leaveTangent=-0.1926, tangentMode=RCTM_User, time=0.4711, value=0.0348}
         <at 1 0>
        ]}}}
     Scale Curve=0.6
    )
    Scale RGB=&MakeVectorFromLinearColorRGB(Color=&ColorFromCurve(ColorCurve=DataInterface() {propertyValues {
        AlphaCurve {keys [<at 0 2 RCIM_Cubic>,<at 0.154 1>,<at 1 1 RCIM_Cubic>]}
        BlueCurve {keys [<at 0.0791 2 RCIM_Cubic>,<at 0.3453 1>,<at 0.6322 1 RCIM_Cubic>]}
        GreenCurve {keys [<at 0.0791 1.6667 RCIM_Cubic>,<at 0.3453 0>,<at 0.6322 0 RCIM_Cubic>]}
        RedCurve {keys [<at 0.0791 0 RCIM_Cubic>,<at 0.3453 0>,<at 0.6322 0.7608 RCIM_Cubic>]}
       }}))
    ScaleA=true
    ScaleRGB=true
   )
   &ScaleRibbonWidth(Ribbon Width Scale=&FloatFromCurve(FloatCurve=DataInterface() {propertyValues {Curve {keys [{arriveTangent=-6.3997, interpMode=RCIM_Cubic, leaveTangent=-2.4825, tangentMode=RCTM_User, time=0, value=1}, <at 1 0>]}}}))
  ]
  rendererClasses [&NiagaraRibbonRendererProperties]
  renderers [{rendererClass=&NiagaraRibbonRendererProperties}]
 }
 BlueBIts {
  simTarget=GPUComputeSim
  emitterUpdateScript [
   &EmitterState()
   &SpawnBurst_Instantaneous(Spawn Count=15, Use Spawn Probability=false)
  ]
  particleSpawnScript [
   /Niagara/Modules/Spawn/Initialization/InitializeParticle(
    Color=&MakeLinearColorFromVectorAndFloat(
     "Vector (RGB)"=&MakeVectorFromLinearColorRGB(Color=<rgba 0 15.4991 30 1>)
     "Float (Alpha)"=&RandomRangeFloat(Minimum=0.005, Maximum=0.05)
    )
    Lifetime=&UniformRangedFloat(Minimum=0.05, Maximum=0.3)
    Module.Write Color=true
    Module.Write Lifetime=true
    Module.Write Mass=true
    Module.Write Material Random=true
    Module.Write Position=false
    Module.Write Scale=false
    Module.Write SpriteRotation=false
    Module.Write SpriteSize=true
    Position=&SimulationPosition()
    Sprite Size=&Vector2DFromFloat(Value=&UniformRangedFloat(Minimum=1, Maximum=4))
   )
   &CylinderLocation(
    Cylinder Height=200
    Cylinder Radius=5
    MaskSpawnGroup=false
    Module.Enable Lathe Profile=false
    Module.Override Randomness=false
    Module.Surface Only=false
    Orientation Axis:ENiagaraOrientationAxis=::XAxis
    Surface Only Band Thickness=0
   )
   &AddVelocity(Velocity=&UniformRangedVector(Minimum=<900 0 0>, Maximum=<2100 0 0>))
  ]
  particleUpdateScript [
   &ParticleState()
   &ScaleSpriteSize(Scale Factor=&Vector2DFromFloat(Value=&FloatFromCurve(FloatCurve=DataInterface() {propertyValues {Curve {keys [{arriveTangent=-5.2934, interpMode=RCIM_Cubic, leaveTangent=-5.2934, tangentMode=RCTM_User, time=0.7, value=1}, <at 1 0>]}}})))
   &SolveForcesAndVelocity(Clamp Velocity=false, Limit Acceleration=false)
   &ScaleSpriteSizeBySpeed(Max Scale Factor=<1 40>, Velocity Threshold=250)
  ]
  rendererClasses [&NiagaraSpriteRendererProperties]
  renderers [{rendererClass=&NiagaraSpriteRendererProperties}]
 }
 LITTLE_RIBBONS {
  emitterUpdateScript [
   &EmitterState()
   &SpawnBurst_Instantaneous(Spawn Count=12, Use Spawn Probability=false)
  ]
  particleSpawnScript [
   &InitializeRibbon(
    Color=<rgba 1 1 1 1.25>
    Lifetime=&RandomRangeFloat(Minimum=0.15, Maximum=0.25, Randomness Mode=::NewEnumerator1, Random Seed=$Emitter.LoopCount, Fixed Random Seed=true)
    Module.Write Color=true
    Module.Write Lifetime=true
    Module.Write Mass=true
    Module.Write Material Random=true
    Module.Write Position=false
    Module.Write RibbonTwist=false
    Module.Write RibbonWidth=true
    Ribbon Width=50
   ) {module=InitializeRibbon001}
   /SetVariables(
    Particles.RibbonID=/Niagara/DynamicInputs/NiagaraID/MakeNiagaraIDFromInt(Index:NiagaraInt32=$Emitter.LoopCount, Acquire Tag:NiagaraInt32=0)
    Particles.RibbonLinkOrder:NiagaraFloat=&ReturnNormalizedExecIndex()
   ) {isSetParameters=true}
   &FN_HermiteCubicSpline(
    Position Blend:NiagaraFloat=$Particles.Initial.RibbonLinkOrder
    Position End:Vector3f=&Add_Vector(
     A=&Multiply_Vector(
      A=&RandomVector(Vector Scale=&RandomRangeFloat(Minimum=10, Maximum=10))
      B=&RandomRangeVector(Minimum=<0 0 0>, Maximum=<0 1 1>)
     )
     B=<50 0 0>
    )
    Position Start:Vector3f=&Add_Vector(
     A=&Multiply_Vector(
      A=&RandomVector(Vector Scale=30)
      B=&RandomRangeVector(Minimum=<1 1 1>, Maximum=<1 1 1>)
     )
     B=<0 0 0>
    )
    Write Particles Position=true
   )
   &AddVelocity(Velocity=&Multiply_Vector(
     A=&VectorFromFloat(Value=&Add_Float(
       A=&Multiply_Float(A=10, B=&ReturnNormalizedExecIndex())
       B=&ReturnNormalizedExecIndex()
      ))
     B=&RandomRangeVector(Minimum=<600 0 0>, Maximum=<600 1 1>, Randomness Mode=::NewEnumerator1, Random Seed=$Emitter.LoopCount, Fixed Random Seed=true)
    )) {module=AddVelocity001}
  ]
  particleUpdateScript [
   &ParticleState()
   &ScaleColor(
    Scale Alpha=&FloatFromCurve(
     FloatCurve=DataInterface() {propertyValues {Curve {keys [{arriveTangent=-5.9757, interpMode=RCIM_Cubic, time=0, value=0}, <at 0.5 0.999 RCIM_Cubic>, <at 1 0>]}}}
     Scale Curve=&FloatFromCurve(
      FloatCurve=DataInterface() {propertyValues {Curve {keys [{arriveTangent=-1.9917, interpMode=RCIM_Cubic, leaveTangent=-1.9917, tangentMode=RCTM_User, time=-0.0108, value=1}, <at 1 0 RCIM_Cubic>]}}}
      CurveIndex=&Multiply_Float(A=$Particles.RibbonLinkOrder, B=$Particles.NormalizedAge)
     )
    )
    Scale RGB=&MakeVectorFromLinearColorRGB(Color=&ColorFromCurve(ColorCurve=DataInterface() {propertyValues {
        AlphaCurve {keys [<at 0 1 RCIM_Cubic>,<at 1 0.5 RCIM_Cubic>]}
        BlueCurve {keys [<at 0.0006 1>,<at 0.9976 1 RCIM_Cubic>]}
        GreenCurve {keys [<at 0.0006 0.6956>,<at 0.9976 0 RCIM_Cubic>]}
        RedCurve {keys [<at 0.0006 0>,<at 0.9976 0.7608 RCIM_Cubic>]}
       }}))
    ScaleA=true
    ScaleRGB=true
   )
   &ScaleRibbonWidth(Ribbon Width Scale=&FloatFromCurve(FloatCurve=DataInterface() {propertyValues {Curve {keys [
         {arriveTangent=-6.3997, interpMode=RCIM_Cubic, time=0, value=1}
         {arriveTangent=-1, interpMode=RCIM_Cubic, leaveTangent=-1, time=0.5, value=1}
         {arriveTangent=-2.2297, interpMode=RCIM_Cubic, leaveTangent=-2.2297, tangentMode=RCTM_User, time=1, value=0}
        ]}}})) {enabled=false}
   &ScaleRibbonWidth(Ribbon Width Scale=&FloatFromCurve(FloatCurve=DataInterface() {propertyValues {Curve {keys [{arriveTangent=-6.3997, interpMode=RCIM_Cubic, time=0, value=0}, {interpMode=RCIM_Cubic, leaveTangent=-2.2297, time=1, value=2}]}}})) {module=ScaleRibbonWidth001}
   &SolveForcesAndVelocity(Clamp Velocity=false, Limit Acceleration=false)
   &DynamicMaterialParameters(
    Index 0 Param 1=1
    Index 0 Param 3=0.1
    Index 1 Param 4=200
    Module.Param0WriteEnabled=true
    Module.Param0WriteEnabled1=true
    Module.Param1WriteEnabled=true
    Module.Param1WriteEnabled1=true
    Module.Param2WriteEnabled=true
    Module.Param2WriteEnabled1=true
    Module.Param3WriteEnabled=true
    Module.Param3WriteEnabled1=true
    Write Parameter Index 1=true
   )
  ]
  rendererClasses [&NiagaraRibbonRendererProperties]
  renderers [{rendererClass=&NiagaraRibbonRendererProperties}]
 }
 LensFlare {
  emitterUpdateScript [
   &EmitterState()
   &SpawnBurst_Instantaneous(Spawn Count=5, Use Spawn Probability=false)
  ]
  particleSpawnScript [
   /SetVariables(Particles.ExecIndex:NiagaraInt32=&ReturnExecIndex()) {isSetParameters=true}
   /SetVariables(Particles.Sprite Spacing=100) {isSetParameters=true}
   /SetVariables() {enabled=false, isSetParameters=true}
   /Niagara/Modules/Spawn/Initialization/V2/InitializeParticle(
    Color=&MakeLinearColorFromVectorAndFloat("Vector (RGB)"=&MakeVectorFromLinearColorRGB(Color=<rgba 0 2 1.8337 1>), "Float (Alpha)"=0.0225)
    Lifetime=0.15
    Position Offset=&MakeVector(
     X=&Add_Float(
      A=-100
      B=&Multiply_Float(
       A=$Particles.Sprite Spacing
       B=&MakeFloatFromInt(Int=&Add_Integer(A=1, B=&ReturnExecIndex()))
      )
     )
     Y=0
     Z=0
    )
    Sprite Size Mode=::NewEnumerator3
    Uniform Sprite Size=&Divide_Float(A=600, B=&MakeFloatFromInt(Int=&Add_Integer(A=1, B=$Particles.ExecIndex)))
    UsePositionOffset=true
    Write Material Random=true
   )
   &DynamicMaterialParameters(
    Index 0 Param 1=1
    Index 0 Param 2=2
    Index 0 Param 3=1
    Index 0 Param 4=500
    Module.Param0WriteEnabled=true
    Module.Param1WriteEnabled=true
    Module.Param2WriteEnabled=true
    Module.Param3WriteEnabled=true
   )
   &CameraOffset(Camera Offset Amount=150)
  ]
  particleUpdateScript [
   &ParticleState()
   &ScaleSpriteSize(Scale Factor=&Vector2DFromFloat(Value=&ClampFloat(Float=&FloatFromCurve(
       FloatCurve=DataInterface() {propertyValues {Curve {keys [
           {arriveTangent=-1.797, interpMode=RCIM_Cubic, leaveTangent=-1.797, tangentMode=RCTM_User, time=0, value=1}
           {interpMode=RCIM_Cubic, leaveTangent=-2.3237, time=1, value=0}
          ]}}}
       Scale Curve=&Add_Float(A=1, B=&MakeFloatFromInt(Int=$Particles.Initial.ExecIndex))
      )))) {enabled=false}
  ]
  rendererClasses [&NiagaraSpriteRendererProperties]
  renderers [{rendererClass=&NiagaraSpriteRendererProperties}]
 }
 Light001 {
  emitterUpdateScript [
   &EmitterState()
   &SpawnBurst_Instantaneous(Spawn Count=1, Use Spawn Probability=false)
  ]
  particleSpawnScript [
   /Niagara/Modules/Spawn/Initialization/InitializeParticle(
    Color=&MultiplyLinearColor(Linear Color=<rgba 0 0.3622 1 1>, Scale Factor=&MakeVector4_FromFloat(Float=1))
    Lifetime=0.1
    Module.Write Color=true
    Module.Write Lifetime=true
    Module.Write Mass=true
    Module.Write Material Random=true
    Module.Write Position=false
    Module.Write Scale=false
    Module.Write SpriteRotation=false
    Module.Write SpriteSize=false
    Position=&SimulationPosition()
   )
   &SystemLocation(MaskSpawnGroup=false, Offset=<0 0 100>, Offset Coordinate Space=::World)
  ]
  particleUpdateScript [
   &ParticleState()
   &ScaleColor(
    Scale Alpha=&FloatFromCurve(FloatCurve=DataInterface() {propertyValues {Curve {keys [<at 0 1 RCIM_Cubic>,<at 1 0>]}}})
    ScaleA=true
    ScaleRGB=true
   )
  ]
  rendererClasses [&NiagaraLightRendererProperties, &NiagaraComponentRendererProperties]
  renderers [
   {rendererClass=&NiagaraLightRendererProperties, rendererIndex=0}
   {rendererClass=&NiagaraComponentRendererProperties, rendererIndex=1}
  ]
 }
 PurpleBits {
  simTarget=GPUComputeSim
  emitterUpdateScript [
   &EmitterState()
   &SpawnBurst_Instantaneous(Spawn Count=10, Use Spawn Probability=false)
  ]
  particleSpawnScript [
   &InitializeParticle(
    Color=&MakeLinearColorFromVectorAndFloat(
     "Vector (RGB)"=&MakeVectorFromLinearColorRGB(Color=<rgba 1.5216 0 2 1>)
     "Float (Alpha)"=&RandomRangeFloat(Minimum=0.4, Maximum=0.65)
    )
    Lifetime=&UniformRangedFloat(Minimum=0.05, Maximum=0.2)
    Module.Write Color=true
    Module.Write Lifetime=true
    Module.Write Mass=true
    Module.Write Material Random=true
    Module.Write Position=false
    Module.Write Scale=false
    Module.Write SpriteRotation=false
    Module.Write SpriteSize=true
    Position=&SimulationPosition()
    Sprite Size=&Vector2DFromFloat(Value=&UniformRangedFloat(Minimum=4, Maximum=7))
   )
   &CylinderLocation(
    Cylinder Height=50
    Cylinder Radius=20
    MaskSpawnGroup=false
    Module.Enable Lathe Profile=false
    Module.Override Randomness=false
    Module.Surface Only=true
    Orientation Axis:ENiagaraOrientationAxis=::XAxis
    Use Endcaps In Surface Only Mode=false
   )
   &AddVelocity(Velocity=&UniformRangedVector(Minimum=<900 0 0>, Maximum=<2100 0 0>))
  ]
  particleUpdateScript [
   &ParticleState()
   &ScaleSpriteSize(Scale Factor=&Vector2DFromFloat(Value=&FloatFromCurve(FloatCurve=DataInterface() {propertyValues {Curve {keys [{arriveTangent=-5.2934, interpMode=RCIM_Cubic, leaveTangent=-5.2934, tangentMode=RCTM_User, time=0.7, value=1}, <at 1 0>]}}})))
   &SolveForcesAndVelocity(Clamp Velocity=false, Limit Acceleration=false)
   &ScaleSpriteSizeBySpeed(Max Scale Factor=<0.5 50>, Velocity Threshold=250)
  ]
  rendererClasses [&NiagaraSpriteRendererProperties]
  renderers [{rendererClass=&NiagaraSpriteRendererProperties}]
 }
 RayGlow_BlueCone {
  emitterUpdateScript [
   &EmitterState()
   &SpawnBurst_Instantaneous(Spawn Count=1, Use Spawn Probability=false)
  ]
  particleSpawnScript [
   &InitializeParticle(
    Color=<rgba 0 0.8333 5 0.4>
    Lifetime=0.1875
    Mesh Scale=&Multiply_Vector(A=<1.25 1.25 1>, B=&VectorFromFloat(Value=3))
    Module.Write Color=true
    Module.Write Lifetime=true
    Module.Write Mass=true
    Module.Write Material Random=true
    Module.Write Position=true
    Module.Write Scale=true
    Module.Write SpriteRotation=false
    Module.Write SpriteSize=false
    Position=&SimulationPosition()
    Sprite Size=<5 5>
   )
   &SystemLocation(MaskSpawnGroup=false)
   &InitialMeshOrientation(Orientation Vector=<0 0 1>, Use Orientation Vector=true, Use Rotation Vector=false) {module=InitialMeshOrientation001}
  ]
  particleUpdateScript [
   &ParticleState()
   &ScaleColor(
    Scale Alpha=&FloatFromCurve(FloatCurve=DataInterface() {propertyValues {Curve {keys [<at 0 1 RCIM_Cubic>,<at 1 0>]}}})
    ScaleA=true
    ScaleRGB=true
   )
  ]
  rendererClasses [&NiagaraMeshRendererProperties]
  renderers [{rendererClass=&NiagaraMeshRendererProperties, rendererIndex=0}]
 }
 RayGlow_PinkCone {
  emitterUpdateScript [
   &EmitterState()
   &SpawnBurst_Instantaneous(Spawn Count=1, Use Spawn Probability=false)
  ]
  particleSpawnScript [
   &InitializeParticle(
    Color=<rgba 0 10 9.4601 1>
    Lifetime=0.36
    Mesh Scale=<0.75 0.75 5>
    Module.Write Color=true
    Module.Write Lifetime=true
    Module.Write Mass=true
    Module.Write Material Random=true
    Module.Write Position=true
    Module.Write Scale=true
    Module.Write SpriteRotation=false
    Module.Write SpriteSize=false
    Position=&SimulationPosition()
    Sprite Size=<5 5>
   )
   &InitialMeshOrientation(Orientation Vector=<0 0 1>, Use Orientation Vector=true, Use Rotation Vector=false) {module=InitialMeshOrientation001}
  ]
  particleUpdateScript [
   &ParticleState()
   &SolveForcesAndVelocity(Clamp Velocity=false, Limit Acceleration=false)
   &ScaleColor(
    Scale Alpha=&FloatFromCurve(
     FloatCurve=DataInterface() {propertyValues {Curve {keys [{arriveTangent=-1.4793, interpMode=RCIM_Cubic, leaveTangent=-1.4793, tangentMode=RCTM_User, time=0, value=1}, <at 1 0>]}}}
     Scale Curve=0.125
    )
    ScaleA=true
    ScaleRGB=true
   )
   &ScaleMeshSize(Scale Factor=&MakeVector(
     X=&Lerp_Float(A=2, B=0, Alpha=$Particles.NormalizedAge)
     Y=&Lerp_Float(A=2, B=0, Alpha=$Particles.NormalizedAge)
     Z=&FloatFromCurve(FloatCurve=DataInterface() {propertyValues {Curve {keys [<at 0 0.25>,<at 1 2>]}}})
    ))
  ]
  rendererClasses [&NiagaraMeshRendererProperties]
  renderers [{rendererClass=&NiagaraMeshRendererProperties, rendererIndex=0}]
 }
 Sprites {
  emitterUpdateScript [
   &EmitterState()
   &SpawnBurst_Instantaneous(Spawn Count=1, Use Spawn Probability=false)
  ]
  particleSpawnScript [
   /Niagara/Modules/Spawn/Initialization/V2/InitializeParticle(
    Color=<rgba 5 15 5 1>
    Lifetime=0.125
    Sprite Size Mode=::NewEnumerator3
    Uniform Sprite Size=60
    UsePositionOffset=false
    Write Material Random=true
   )
   /NiagaraFN/Sprite/FN_SpriteAlignmentAndFacing(SpriteAlignment:Vector3f=<0 1 0>, SpriteFacing:Vector3f=<0 0 1>)
   &SystemLocation(MaskSpawnGroup=false, Offset=<0 0 4>, Offset Coordinate Space=::Simulation)
  ]
  particleUpdateScript [
   &ParticleState()
   &ScaleColor(
    Linear Color Curve=DataInterface() {propertyValues {
      AlphaCurve {keys [<at -0.004 0>,<at 0.2723 1>,<at 0.7814 1>,<at 1 0>]}
      BlueCurve {keys [<at 0 1>,<at 0.1 1>,<at 0.3 1>,<at 0.9281 1>]}
      GreenCurve {keys [<at 0 0.0521>,<at 0.1 0.1454>,<at 0.3 0.4087>,<at 0.9281 0.0469>]}
      RedCurve {keys [<at 0 0.7732>,<at 0.1 0.0677>,<at 0.3 0.0729>,<at 0.9281 0.772>]}
     }}
    Scale Mode=::NewEnumerator2
   )
  ]
  rendererClasses [&NiagaraSpriteRendererProperties]
  renderers [{rendererClass=&NiagaraSpriteRendererProperties}]
 }
}
summary {
 systemName=NS_DarkTower_Boost_Trigger
 emitters [
  {emitterName=RayGlow_PinkCone, rendererClasses [&NiagaraMeshRendererProperties]}
  {emitterName=RayGlow_BlueCone, rendererClasses [&NiagaraMeshRendererProperties]}
  {emitterName=BlueBIts, simTarget=GPUComputeSim, rendererClasses [&NiagaraSpriteRendererProperties]}
  {emitterName=PurpleBits, simTarget=GPUComputeSim, rendererClasses [&NiagaraSpriteRendererProperties]}
  {emitterName=Sprites, rendererClasses [&NiagaraSpriteRendererProperties]}
  {bEnabled=false, emitterName=BIG_RIBBON, rendererClasses [&NiagaraRibbonRendererProperties]}
  {emitterName=LITTLE_RIBBONS, rendererClasses [&NiagaraRibbonRendererProperties]}
  {emitterName=LensFlare, rendererClasses [&NiagaraSpriteRendererProperties]}
  {emitterName=Light001, rendererClasses [&NiagaraLightRendererProperties, &NiagaraComponentRendererProperties]}
 ]
 userVariables []
}
